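# Physics asset for the face accessory mesh of the F_MED_Twin_Dark female player body, defining bodies and joints for dynamic hair bones.
# Hair bones use single capsules, with short stubs jointed directly to the head, plus a six-link left strand and a two-link right tip chained from it.
# Swing limits are narrow with very tight twist, and mostly widen toward the left strand's end.
# Two head capsules and a large spine_05 capsule form the body volumes, linked by a free-rotation joint.
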PhysicsAsset /BRCosmetics/Characters/Player/Female/Medium/Bodies/F_MED_Twin_Dark/Meshes/FaceAcc/F_MED_Twin_Dark_FaceAcc_Physics
bodies {
 dyn_hair_strand_01_l [{shapeType=Capsule, length=2, radius=2.25, center=<0.5 0 0>, rotation=<rot 90 0 0>}]
 dyn_hair_strand_02_l [{shapeType=Capsule, length=2.75, radius=1.5, center=<0.8 0 0>, rotation=<rot 90 0 0>}]
 dyn_hair_strand_03_l [{shapeType=Capsule, length=3.25, radius=1, center=<0.8 0 0>, rotation=<rot 90 0 0>}]
 dyn_hair_strand_04_l [{shapeType=Capsule, length=3.25, radius=1, center=<0.9 0 0>, rotation=<rot 90 0 0>}]
 dyn_hair_strand_05_l [{shapeType=Capsule, length=2.75, radius=1, center=<1 0 0>, rotation=<rot 90 0 0>}]
 dyn_hair_strand_06_l [{shapeType=Capsule, length=2.25, radius=1, center=<1.5 0 0>, rotation=<rot 82.9759 0 0.0001>}]
 dyn_hair_stubA_l [{shapeType=Capsule, length=1.25, radius=1.25, center=<2 0 0>, rotation=<rot 90 0 0>}]
 dyn_hair_stubA_r [{shapeType=Capsule, length=3, radius=1.75, center=<2 0 0>, rotation=<rot 90 0 0>}]
 dyn_hair_stubB_l [{shapeType=Capsule, length=2, radius=1.25, center=<2 0 0>, rotation=<rot 90 0 0>}]
 dyn_hair_stubB_r [{shapeType=Capsule, length=1.75, radius=1.5, center=<2 0 0>, rotation=<rot 90 0 0>}]
 dyn_hair_stubC_l [{shapeType=Capsule, length=2.5, radius=1.25, center=<2 0 0>, rotation=<rot 90 0 0>}]
 dyn_hair_stub_bck [{shapeType=Capsule, length=2.6825, radius=2.1955, center=<1.3807 0 0.3053>, rotation=<rot 78.2949 179.9998 180>}]
 dyn_hair_stub_bck_l [{shapeType=Capsule, length=1.8919, radius=1.5125, center=<0.9308 0 0.1213>, rotation=<rot 72.141 -179.9998 -179.9998>}]
 dyn_hair_stub_bck_r [{shapeType=Capsule, length=1.8919, radius=1.2625, center=<0.9831 0 0.5286>, rotation=<rot 62.807 -179.9998 -179.9998>}]
 dyn_hair_tip_01_r [{shapeType=Capsule, length=1, radius=1, center=<0.5 0 0>, rotation=<rot 90 0 0>}]
 dyn_hair_tip_02_r [{shapeType=Capsule, length=1.25, radius=0.75, center=<1 0 0>, rotation=<rot 90 0 0>}]
 head [
  {shapeType=Capsule, length=8.5, radius=6.25, center=<1.5028 -0.1612 -0.0001>, rotation=<rot -90 -90 83.8767>}
  {shapeType=Capsule, length=10.5, radius=6.25, center=<0.4869 2.6145 0>, rotation=<rot -90 0 13.2232>}
 ]
 spine_05 [{shapeType=Capsule, length=18.5, radius=36.755, center=<-19.408 -22.3695 -0.0001>, rotation=<rot 0 -90 0>}]
}
constraints [
 {bone1Name=dyn_hair_stubA_l, bone2Name=head, swing1LimitDegrees=15, swing2LimitDegrees=15, twistLimitDegrees=2}
 {bone1Name=dyn_hair_stubB_l, bone2Name=head, swing1LimitDegrees=15, swing2LimitDegrees=15, twistLimitDegrees=2}
 {bone1Name=dyn_hair_stubC_l, bone2Name=head, swing1LimitDegrees=15, swing2LimitDegrees=15, twistLimitDegrees=2}
 {bone1Name=dyn_hair_stub_bck_l, bone2Name=head, swing1LimitDegrees=15, swing2LimitDegrees=25, twistLimitDegrees=2}
 {bone1Name=dyn_hair_stub_bck, bone2Name=head, swing1LimitDegrees=15, swing2LimitDegrees=40, twistLimitDegrees=2}
 {bone1Name=dyn_hair_stub_bck_r, bone2Name=head, swing1LimitDegrees=15, swing2LimitDegrees=25, twistLimitDegrees=2}
 {bone1Name=dyn_hair_stubA_r, bone2Name=head, swing1LimitDegrees=35.5714, swing2LimitDegrees=21.8571, twistLimitDegrees=2}
 {bone1Name=dyn_hair_stubB_r, bone2Name=head, swing1LimitDegrees=21.8571, swing2LimitDegrees=20.1429, twistLimitDegrees=2}
 {bone1Name=dyn_hair_strand_01_l, bone2Name=head, swing1LimitDegrees=15, swing2LimitDegrees=15, twistLimitDegrees=3}
 {bone1Name=dyn_hair_strand_02_l, bone2Name=dyn_hair_strand_01_l, swing1LimitDegrees=15, swing2LimitDegrees=9.7143, twistLimitDegrees=3}
 {bone1Name=dyn_hair_strand_03_l, bone2Name=dyn_hair_strand_02_l, swing1LimitDegrees=16, swing2LimitDegrees=12.8571, twistLimitDegrees=3}
 {bone1Name=dyn_hair_strand_04_l, bone2Name=dyn_hair_strand_03_l, swing1LimitDegrees=18, swing2LimitDegrees=20, twistLimitDegrees=3}
 {bone1Name=dyn_hair_strand_05_l, bone2Name=dyn_hair_strand_04_l, swing1LimitDegrees=22, swing2LimitDegrees=22, twistLimitDegrees=3}
 {bone1Name=dyn_hair_strand_06_l, bone2Name=dyn_hair_strand_05_l, swing1LimitDegrees=25, swing2LimitDegrees=25, twistLimitDegrees=3}
 {
  bone1Name=head
  bone2Name=spine_05
  swing1LimitDegrees=45
  swing1Motion=Free
  swing2LimitDegrees=45
  swing2Motion=Free
  twistLimitDegrees=45
  twistMotion=Free
 }
 {bone1Name=dyn_hair_tip_01_r, bone2Name=head, swing1LimitDegrees=23.5714, swing2LimitDegrees=18.4286, twistLimitDegrees=2}
 {bone1Name=dyn_hair_tip_02_r, bone2Name=dyn_hair_tip_01_r, swing1LimitDegrees=21.8571, swing2LimitDegrees=21.8571, twistLimitDegrees=2}
]
